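; Niagara-based camera lens effect Blueprint for Masamune's big slash, pairing a screen particle effect with a post-process pulse.
; When the Niagara component activates, a timeline plays and drives the PostProcess component's blend weight from its float track.
; Each timeline update also re-sends a fixed LifeTime and the Color variable, as ScreenColor, to the Niagara system.
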(event OnComponentActivated(NiagaraComponent)
  (params (Component :type Actor%20Component%20Object%20Reference) (bReset :type Boolean))
  (bind _niagaracomponent (Variables|NiagaraActor|GetNiagaraComponent))
  (bind (_direction _newtrack_0) (Timeline|Timeline Play)
    (:Update
      (Variables|/Script/Engine.PostProcessComponent|SetBlendWeight _newtrack_0 (Variables|Default|GetPostProcess))
      (Niagara|SetNiagaraVariable(Float) _niagaracomponent "LifeTime" 0.8)
      (Niagara|SetNiagaraVariable(LinearColor) _niagaracomponent "ScreenColor" (Variables|Default|GetColor)))))

(fn ConstructionScript
  (|CallParentFunction))
