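# Translucent unlit material for Niagara mesh particles, by its name a camera effect for Masamune, drawing scrolling smoky streaks.
# Particle color RGB passes through an emissive adjustment function.
# The alpha chain sums panning smoke, a V ramp and noise sharpened by a gradient-driven exponent, then masks them with a V falloff and particle alpha.
# Several further noise branches feed nothing visible in the graph.
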Material /MasamuneGameplay/Assets/FX/NewFX/Materials/M_M_Masamune_CameraEffect
graph {
 properties {
  bAutomaticallySetUsageInEditor=false
  bUsedWithNiagaraMeshParticles=true
  blendMode=BLEND_Translucent
  shadingModel=MSM_Unlit
  usedShadingModels=Unlit
 }
 nodes [
  n1 = ParticleColor()
  MaterialFunctionCall(Emissive=@n1.RGB) {materialFunction=&MF_FX_EmissiveAdjustments}
  n2 = ComponentMask(None=TextureCoordinate()) {g=true}
  n3 = Multiply(A=Power(Base=OneMinus(None=@n2)) {constExponent=3}) {constB=3}
  n4 = TextureSample(UVs=Panner(Coordinate=TextureCoordinate() {vTiling=3}) {speedY=4}) {texture=/Game/Effects/Fort_Effects/Textures/Smoke/T_Smoke_Trail_Soft_2}
  n5 = TextureCoordinate()
  n6 = Add(A=Multiply(A=Saturate(None=@n3), B=Multiply(A=@n4.B) {constB=6}) {constA=1}, B=Multiply(A=ComponentMask(None=@n5) {g=true}) {constB=0.5})
  n7 = Panner(Coordinate=TextureCoordinate() {uTiling=4, vTiling=0.5}) {speedY=1}
  n8 = TextureCoordinate()
  n9 = Power(Base=OneMinus(None=ComponentMask(None=@n8) {g=true})) {constExponent=1}
  n10 = Multiply(A=Power(
    Base=TextureSample(UVs=@n7) {texture=/Game/Effects/Fort_Effects/GameplayPulginsAssets/Masamune/T_Noise04}.R
    Exp=LinearInterpolate(Alpha=@n9) {constB=3}
   ))
  n11 = Multiply(A=Power(Base=Saturate(None=@n10)) {constExponent=5}) {constB=1.3}
  n12 = ComponentMask(None=TextureCoordinate()) {g=true}
  n13 = Multiply(A=Power(Base=OneMinus(None=@n12)) {constExponent=4.2}) {constB=3}
  Multiply(A=@n1.A, B=Multiply(A=Add(A=@n6, B=@n11), B=Saturate(None=@n13)))
  Panner(Coordinate=TextureCoordinate() {uTiling=5, vTiling=0.5}) {speedY=0.5}
  n14 = Panner(Coordinate=TextureCoordinate() {uTiling=4, vTiling=0.5}) {speedY=0.5}
  n15 = TextureSample(UVs=@n14) {texture=/Game/Weapons/FORT_Melee/Pickaxe_CoatCheck/FX/T_Pickaxe_CoatCheck_Noise}
  n16 = TextureSample(UVs=@n14) {samplerType=SAMPLERTYPE_Grayscale, texture=&LowResBlurredNoise}
  Power(Base=Multiply(A=LinearInterpolate(A=@n15.R, B=@n16.R)) {constB=2})
  n17 = Panner(Coordinate=TextureCoordinate() {uTiling=3, vTiling=0.2}) {speedY=1}
  Saturate(None=LinearInterpolate(Alpha=TextureSample(UVs=@n17) {texture=&T_Pickaxe_CoatCheck_Noise}.R) {constA=0.2, constB=2})
  TextureSample() {texture=/Game/Accessories/FORT_Backpacks/Backpack_M_MED_AllKnowing/FX/T_Backpack_AllKnowing_Noise_02}
 ]
}
