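# Unlit, two-sided translucent material for Niagara mesh particles, named for the lingering airflow of the Masamune air attack.
# Particle color feeds an emissive adjustment function, while SplineThicken takes mesh width from the SplineWidth dynamic parameter.
# The Panner dynamic parameter replaces time to scroll a clamped ribbon mask along U.
# That mask is multiplied by blurred noise, a sine along U, a U-end fade and particle alpha.
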Material /MasamuneGameplay/Assets/FX/NewFX/Materials/M_Masamune_AirAttack_RemainAirflow
graph {
 properties {
  bAutomaticallySetUsageInEditor=false
  bUsedWithNiagaraMeshParticles=true
  blendMode=BLEND_Translucent
  pixelDepthOffsetMode=PDOM_Legacy
  previewMesh=/Game/Effects/Fort_Effects/Mesh/Splines/SM_Spiral_Spline
  shadingModel=MSM_Unlit
  twoSided=true
  usedShadingModels=Unlit
 }
 nodes [
  Constant() {r=50}
  n1 = ParticleColor()
  MaterialFunctionCall(Emissive=@n1.RGB) {materialFunction=&MF_FX_EmissiveAdjustments}
  n2 = DynamicParameter() {defaultValue=<rgba 0 50 1 1>, paramNames [Panner,SplineWidth,Param3,Param4]}
  MaterialFunctionCall(
   WidthBase=@n2.SplineWidth
   WidthTip=@n2.SplineWidth
   Expand U or V UV Channel=StaticBoolParameter() {parameterName="Expand U or V UV Channel (SB)"}
  ) {materialFunction=&SplineThicken}
  MaterialFunctionCall() {materialFunction=&TimeWithSpeedVariable}
  n3 = TextureCoordinate() {uTiling=1.5, vTiling=0.7}
  n4 = Constant2Vector() {g=0.2, r=1}
  n5 = DynamicParameter() {defaultValue=<rgba 0 50 1 1>, paramNames [Panner,SplineWidth,Param3,Param4]}
  n6 = TextureSample(UVs=Panner(Coordinate=Add(A=@n3, B=@n4), Time=LinearInterpolate(Alpha=@n5.Panner) {constB=8}) {speedX=-0.5}) {
   samplerSource=SSM_Clamp_WorldGroupSettings
   samplerType=SAMPLERTYPE_LinearGrayscale
   texture=/Game/Gadgets/Assets/VinderTech_GliderChute/Glider_Ensemble_Maroon_Male/FX/Textures/T_RibbonMask_B
  }
  n7 = Multiply(
   A=@n6.R
   B=TextureSample(UVs=TextureCoordinate() {uTiling=1.5}) {samplerType=SAMPLERTYPE_Grayscale, texture=&LowResBlurredNoise}.R
  )
  n8 = Sine(None=ComponentMask(None=TextureCoordinate()) {r=true}) {period=2}
  n9 = TextureCoordinate()
  n10 = Multiply(A=OneMinus(None=ComponentMask(None=@n9) {r=true})) {constB=2}
  Multiply(A=@n1.A, B=Multiply(A=Multiply(A=@n7, B=@n8), B=Saturate(None=@n10)))
 ]
}
