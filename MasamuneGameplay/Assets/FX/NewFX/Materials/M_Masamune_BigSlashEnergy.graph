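# Translucent, unlit, two-sided material for Niagara mesh particles, evidently the energy of Masamune's big slash effect.
# Emissive adds two scrolling noise and caustic textures, UV-distorted by panning blurred noise, tints them by particle color and passes through MF_FX_EmissiveAdjustments.
# Opacity combines scrolling distorted noise, sine envelopes along U and V, and particle alpha.
# A view-angle fade, with sensitivity driven by vertex-color red, also scales opacity.
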Material /MasamuneGameplay/Assets/FX/NewFX/Materials/M_Masamune_BigSlashEnergy
graph {
 properties {
  bAutomaticallySetUsageInEditor=false
  bUsedWithNiagaraMeshParticles=true
  blendMode=BLEND_Translucent
  shadingModel=MSM_Unlit
  twoSided=true
  usedShadingModels=Unlit
 }
 nodes [
  Constant() {r=2}
  n1 = TextureCoordinate()
  n2 = Multiply(A=TextureSample(UVs=Panner() {speedY=0.1}) {samplerType=SAMPLERTYPE_Grayscale, texture=&LowResBlurredNoise}.R) {constB=0.2}
  n3 = TextureSample(UVs=Panner(Coordinate=Add(A=@n1, B=@n2)) {speedX=-0.5}) {texture=/Game/Effects/Fort_Effects/Textures/Tile/T_Noise_03}
  Desaturation(None=@n3.G, Fraction=Constant() {r=-3})
  n4 = ParticleColor()
  n5 = Add(A=TextureCoordinate() {uTiling=0.7}, B=@n2)
  n6 = Power(Base=TextureSample(UVs=Panner(Coordinate=@n5) {speedX=-0.3}) {samplerType=SAMPLERTYPE_LinearGrayscale, texture=/Game/Global/Textures/Default/Caustics/T_Nosie_Caustics_Grayscale}.R) {constExponent=1}
  MaterialFunctionCall(Emissive=Multiply(A=@n4.RGB, B=Add(A=@n3.G, B=@n6))) {materialFunction=&MF_FX_EmissiveAdjustments}
  n7 = Add(
   A=TextureCoordinate() {uTiling=0.5}
   B=Multiply(A=TextureSample() {samplerType=SAMPLERTYPE_Grayscale, texture=&LowResBlurredNoise}.R) {constB=0.2}
  )
  n8 = TextureCoordinate()
  n9 = Multiply(
   A=TextureSample(UVs=Panner(Coordinate=@n7) {speedX=0.5}) {samplerType=SAMPLERTYPE_Grayscale, texture=&LowResBlurredNoise}.R
   B=Sine(None=ComponentMask(None=@n8) {r=true}) {period=2}
  ) {constA=1}
  n10 = TextureCoordinate()
  n11 = Power(Base=Sine(None=ComponentMask(None=@n10) {g=true}) {period=2})
  n12 = VertexColor()
  Multiply(
   A=Multiply(A=@n4.A, B=Multiply(A=@n9, B=@n11))
   B=MaterialFunctionCall("FadeSensitivity (optional)"=LinearInterpolate(Alpha=@n12.R) {constB=5}) {materialFunction=&FadeBasedOnViewAngle}.Result
  )
 ]
}
