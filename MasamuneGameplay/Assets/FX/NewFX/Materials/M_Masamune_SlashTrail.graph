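# Translucent, unlit, two-sided slash trail material for Masamune effects on Niagara ribbons and mesh particles.
# Opacity combines a noise-distorted slash mask with sine edge fades, view-angle fade, HeightLerp dissolve and depth fade.
# Color multiplies two curve-atlas gradients along U and V with particle color and panning flame noise, then adds a rainbow-texture tint.
# Niagara dynamic parameters drive dissolve, mask offset, desaturation and the width of optional SplineThicken world position offset.
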Material /MasamuneGameplay/Assets/FX/NewFX/Materials/M_Masamune_SlashTrail
graph {
 properties {
  bAutomaticallySetUsageInEditor=false
  bUsedWithNiagaraMeshParticles=true
  bUsedWithNiagaraRibbons=true
  blendMode=BLEND_Translucent
  shadingModel=MSM_Unlit
  twoSided=true
  usedShadingModels=Unlit
 }
 nodes [
  n1 = ComponentMask(None=TextureCoordinate()) {r=true}
  n2 = OneMinus(None=@n1)
  n3 = Multiply(A=Sine(None=@n1) {period=2}, B=LinearInterpolate(B=Power(Base=@n2), Alpha=ScalarParameter() {parameterName=UseTailFade}) {constA=1})
  n4 = VectorParameter() {parameterName=UVScale, channelNames {a="", b="", g=V Scale, r=U Scale}, defaultValue=<rgba 0.3 1 0 0>}
  n5 = Multiply(A=TextureCoordinate(), B=ComponentMask(None=@n4) {g=true, r=true})
  n6 = VectorParameter() {parameterName=UVOffset, channelNames {a="", b="", g=V Offset, r=U Offset}, defaultValue=<rgba 0.35 0 0 0>}
  n7 = ComponentMask(None=@n6) {g=true, r=true}
  n8 = VectorParameter() {parameterName=DisturPanner, defaultValue=<rgba -1 0 0 0>}
  n9 = TextureSample(UVs=Panner(Coordinate=TextureCoordinate() {vTiling=0.7}, Speed=ComponentMask(None=@n8.RGB) {g=true, r=true}) {speedX=1}) {samplerType=SAMPLERTYPE_Grayscale, texture=&LowResBlurredNoise}
  n10 = ScalarParameter() {parameterName=Disturb}
  n11 = Panner(
   Coordinate=Add(A=Add(A=@n5, B=@n7), B=Multiply(A=@n9.R, B=@n10) {constB=0.2})
   Time=DynamicParameter() {defaultValue=<rgba 0 0.2 0 1>, paramNames [Dissolve, DissolveOffset, MaskOffset, Param4]}.MaskOffset
   Speed=ComponentMask(None=VectorParameter() {parameterName=MaskPanner}.RGB) {g=true, r=true}
  )
  n12 = TextureCoordinate() {uTiling=0.5}
  n13 = Multiply(
   A=StaticComponentMaskParameter(None=TextureSampleParameter2D(UVs=@n11) {parameterName=Defalut Tex, texture=/Game/Effects/Fort_Effects/GameplayPulginsAssets/Masamune/T_Masamune_Slash}.RGB) {defaultR=true, parameterName=Mask}
   B=LinearInterpolate(
    B=TextureSample(UVs=@n12) {samplerType=SAMPLERTYPE_Grayscale, texture=&LowResBlurredNoise}.R
    Alpha=ScalarParameter() {parameterName=UseDetailMask}
   ) {constA=1}
  )
  n14 = Multiply(
   A=LinearInterpolate(Alpha=VertexColor().R) {constA=2, constB=2}
   B=ScalarParameter() {defaultValue=1, parameterName="FadeSensitivity (optional) (S)"}
  )
  n15 = DynamicParameter() {defaultValue=<rgba 0 0.2 0 1>, paramNames [Dissolve, DissolveOffset, MaskOffset, Param4]}
  n16 = ScalarParameter() {parameterName=DissolveMin}
  n17 = ScalarParameter() {defaultValue=1, parameterName=DissolveMax}
  n18 = TextureCoordinate()
  n19 = VectorParameter() {parameterName=Dissolve UV, channelNames {a="", b="", g=V scale, r=U scale}, defaultValue=<rgba 1.5 1 0 0>}
  n20 = ComponentMask(None=@n19) {g=true, r=true}
  n21 = VectorParameter() {parameterName=Dissolve Speed, channelNames {a="", b="", g=V scale, r=U scale}, defaultValue=<rgba -0.5 -0.5 0 0>}
  n22 = TextureSampleParameter2D(UVs=Panner(Coordinate=Multiply(A=@n18, B=@n20), Time=@n15.DissolveOffset, Speed=ComponentMask(None=@n21) {g=true, r=true}) {speedX=-0.5, speedY=-0.5}) {parameterName=Dissolve_Tex, texture=/Game/CharClasses/Outlander/Gadgets/AntiMaterialCharge/Materials/T_AntiMaterialCharge_Dissolve_01}
  n23 = MaterialFunctionCall(Transition Phase=@n15.Dissolve, Height Texture=Saturate(None=LinearInterpolate(A=@n16, B=@n17, Alpha=@n22.R))) {materialFunction=&HeightLerp}
  n24 = ScalarParameter() {defaultValue=10, parameterName=Dissolve edge shape}
  n25 = Multiply(
   A=Multiply(
    A=Multiply(A=@n13, B=@n3)
    B=MaterialFunctionCall("FadeSensitivity (optional)"=@n14) {materialFunction=&FadeBasedOnViewAngle}.Result
   )
   B=Saturate(None=Multiply(A=@n23.Results, B=@n24) {constB=10})
  )
  n26 = Sine(None=ComponentMask(None=TextureCoordinate()) {g=true}) {period=2}
  DepthFade(
   Opacity=Multiply(A=Multiply(A=@n25, B=@n26), B=ParticleColor().A)
   FadeDistance=ScalarParameter() {defaultValue=50, parameterName=FadeDistance}
  ) {fadeDistanceDefault=50}
  n27 = Add(A=TextureCoordinate() {vTiling=0.5}, B=Constant2Vector() {r=0.7})
  n28 = ComponentMask(None=VectorParameter() {parameterName=ColorTexPanner}.RGB) {g=true, r=true}
  n29 = Power(Base=TextureSample(UVs=Panner(Coordinate=@n27, Speed=@n28)) {texture=&T_Tile_RainbowColor}.G) {constExponent=3}
  Desaturation(None=@n29, Fraction=ScalarParameter() {defaultValue=-0.7, parameterName=ColorDes})
  n30 = DynamicParameter() {parameterIndex=2, defaultValue=<rgba 10 1 0 1>, paramNames [Width,DesScale,none,Param4]}
  LinearInterpolate(
   B=MaterialFunctionCall(WidthBase=@n30.Width, WidthTip=@n30.Width, Expand U or V UV Channel=StaticBool()) {materialFunction=&SplineThicken}.WorldPositionOffset
   Alpha=ScalarParameter() {parameterName=UseSplineThicken}
  )
  n31 = LinearInterpolate(
   B=Multiply(A=Constant3Vector() {constant=<rgba 1 0.9201 0.5208 1>}, B=@n29)
   Alpha=ScalarParameter() {defaultValue=1, parameterName=AddedColorful}
  )
  n32 = ParticleColor()
  n33 = TextureCoordinate()
  n34 = CurveAtlasRowParameter(CurveTime=ComponentMask(None=@n33) {g=true}) {
   atlas=/MasamuneGameplay/Assets/FX/NewFX/CAtlas_Masamune_Color
   curve=/MasamuneGameplay/Assets/FX/NewFX/CV_Masamune_SlashColor
   parameterName=Color
  }
  n35 = ComponentMask(None=@n33) {r=true}
  n36 = CurveAtlasRowParameter(CurveTime=@n35) {
   atlas=&CAtlas_Masamune_Color
   curve=/MasamuneGameplay/Assets/FX/NewFX/CV_Masamune_SlashColor2
   defaultValue=2
   parameterName=Color2
  }
  n37 = Multiply(A=@n34, B=@n36)
  n38 = ScalarParameter() {defaultValue=1, parameterName=MainColorIntensity}
  n39 = Multiply(A=@n32.RGB, B=Add(B=Multiply(A=@n37, B=@n38)))
  n40 = VectorParameter() {parameterName=Default_Tex_Panner, defaultValue=<rgba 0.5 0 0 0>}
  n41 = TextureSample(UVs=Panner(Coordinate=TextureCoordinate() {uTiling=0.5, vTiling=0.5}, Speed=ComponentMask(None=@n40.RGB) {g=true, r=true}) {speedX=0.5}) {samplerType=SAMPLERTYPE_LinearGrayscale, texture=/Game/Weapons/FORT_Melee/Pickaxe_Cube_Ninja_Male/FX/T_FluidFlameTile}
  n42 = LinearInterpolate(
   B=LinearInterpolate(B=ScalarParameter() {defaultValue=2, parameterName=Detail Noise Max Intensity}, Alpha=Power(Base=@n41.R) {constExponent=1}) {constA=0.7, constB=2}
   Alpha=ScalarParameter() {defaultValue=1, parameterName=UseDetail Noise}
  ) {constA=1}
  n43 = LinearInterpolate(A=Multiply(A=Multiply(A=@n39), B=@n42), B=Multiply(A=@n32.RGB, B=@n42), Alpha=ScalarParameter() {parameterName=Use Simple Color})
  n44 = ScalarParameter() {parameterName=Des}
  n45 = DynamicParameter() {parameterIndex=2, defaultValue=<rgba 10 1 0 1>, paramNames [Width,DesScale,none,Param4]}
  MaterialFunctionCall(Emissive=Desaturation(None=Add(A=@n31, B=@n43), Fraction=Multiply(A=@n44, B=@n45.DesScale))) {materialFunction=&MF_FX_EmissiveAdjustments}
  MaterialFunctionCall() {materialFunction=&DebugTimeSine}
  OneMinus(None=@n35)
  ScalarParameter() {defaultValue=10, parameterName="WidthBase (S)"}
  ScalarParameter() {defaultValue=10, parameterName="WidthTip (S)"}
  StaticBoolParameter() {parameterName="Expand U or V UV Channel (SB)"}
 ]
}
