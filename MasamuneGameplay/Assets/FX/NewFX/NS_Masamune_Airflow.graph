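# Niagara effect for Masamune: a short, soft airflow puff built from a single mesh-particle emitter.
# A small burst spawns randomly scaled and rotated meshes along a zero-radius vertical cylinder.
# Over each brief lifetime they grow with extra vertical stretch, spin around Z and fade in then out.
# Time-of-day functions tint the color and feed a dynamic material parameter, alongside a curve-driven parameter that ramps late in life.
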NiagaraSystem /MasamuneGameplay/Assets/FX/NewFX/NS_Masamune_Airflow
dependencies {usedDataInterfaces [&NiagaraDataInterfaceCurve], usedDynamicInputs [], usedModules [&SystemState]}
emitters {SoftAirflow {
  emitterUpdateScript [
   &EmitterState()
   &SpawnBurst_Instantaneous(Spawn Count=5, Use Spawn Probability=false, UseLoopCountLimit=false)
  ]
  particleSpawnScript [
   &InitializeParticle(
    Color=<rgba 0.8698 0.8698 0.8698 1>
    Lifetime Max=0.8
    Lifetime Min=0.4
    Lifetime Mode=::NewEnumerator1
    Mesh Scale=&Multiply_VectorByFloat(Vector=<1 1 1>, Float=&RandomRangeFloat(Minimum=2, Maximum=3))
    Mesh Scale Mode=::NewEnumerator4
    Position Offset=<0 0 200>
    UsePositionOffset=false
    Write Material Random=true
   )
   &InitialMeshOrientation(
    Mesh Orientation Mode=::NewEnumerator7
    Rotation=&RandomRangeVector(Minimum=<0 0 -1>, Maximum=<0 0 1>)
    Use Rotation Vector=true
   )
   &ShapeLocation(
    Cylinder Height=70
    Cylinder Radius=0
    MaskSpawnGroup=false
    Override Seed=false
    Shape Primitive=::NewEnumerator1
    Surface Only=false
    UseOwnerScale=true
   )
  ]
  particleUpdateScript [
   &ParticleState()
   &DynamicMaterialParameters(
    Index 0 Param 1=&FloatFromCurve(FloatCurve=DataInterface() {propertyValues {Curve {keys [{interpMode=RCIM_Cubic, tangentMode=RCTM_User, time=0.5, value=0}, {interpMode=RCIM_Cubic, tangentMode=RCTM_User, time=1, value=1}]}}})
    Index 0 Param 2=&FNTimeOfDayScalar(Night Scale=0, Morning Scale=1, Evening Scale=1)
    Index 0 Param 3=1
    Module.Param0WriteEnabled=true
    Module.Param1WriteEnabled=true
    Module.Param2WriteEnabled=true
    Module.Param3WriteEnabled=true
   )
   &ScaleMeshSize(Scale Factor=&Multiply_VectorByFloat(
     Vector=<1 1 1.5>
     Float=&FloatFromCurve(
      FloatCurve=DataInterface() {propertyValues {Curve {keys [
          {arriveTangent=2.376, interpMode=RCIM_Cubic, leaveTangent=2.376, tangentMode=RCTM_User, time=0, value=0}
          {arriveTangent=0.1286, interpMode=RCIM_Cubic, leaveTangent=0.1286, tangentMode=RCTM_User, time=1, value=1}
         ]}}}
      Scale Curve=1.1
     )
    ))
   &UpdateMeshOrientation(Rotation Vector=<0 0 1>)
   &ScaleColor(
    Scale Alpha=&FloatFromCurve(
     FloatCurve=DataInterface() {propertyValues {Curve {keys [
         {arriveTangent=1.8077, interpMode=RCIM_Cubic, leaveTangent=1.8077, tangentMode=RCTM_User, time=0, value=0}
         {arriveTangent=-1.9397, interpMode=RCIM_Cubic, leaveTangent=-2.9533, tangentMode=RCTM_Break, time=0.5, value=0.5}
         {arriveTangent=-0.7636, interpMode=RCIM_Cubic, leaveTangent=-0.7636, tangentMode=RCTM_User, time=1, value=0}
        ]}}}
     Scale Curve=0.5
    )
    Scale RGB=&MakeVectorFromLinearColorRGB(Color=&FNTimeOfDayColor(Night Color=<rgba 0.3392 0.4002 0.6 1>, Morning Color=<rgba 0.7 0.7 0.7 1>, Day Color=<rgba 1 1 1 1>, Evening Color=<rgba 0.7 0.7 0.7 1>))
    ScaleA=true
    ScaleRGB=true
   )
  ]
  rendererClasses [&NiagaraMeshRendererProperties]
  renderers [{rendererClass=&NiagaraMeshRendererProperties, rendererIndex=0}]
 }}
summary {
 systemName=NS_Masamune_Airflow
 emitters [{emitterName=SoftAirflow, rendererClasses [&NiagaraMeshRendererProperties]}]
 userVariables []
}
