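# Niagara camera lens effect for Masamune's sprint, pairing a screen wave with a short speedline burst.
# ScreenWave spawns a mesh tinted by User.ScreenColor, animating its fade and dynamic material parameters over particle life.
# Speedlines spawns thin GPU sprites with point-based velocity on a stretched ring, using emitter loop-age curves to pulse spawn rate, ring radius and opacity.
# User.LifeTime is exposed but unreferenced in the listed modules; durations are fixed per emitter.
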NiagaraSystem /MasamuneGameplay/Assets/FX/NewFX/NS_Masamune_Camera_Lens_Sprint
dependencies {
 usedDataInterfaces [&NiagaraDataInterfaceColorCurve, &NiagaraDataInterfaceCurve, &NiagaraDataInterfaceVector2DCurve]
 usedDynamicInputs []
 usedModules [&SystemState]
}
emitters {
 ScreenWave {
  emitterUpdateScript [
   &EmitterState(Life Cycle Mode=::NewEnumerator1, Loop Behavior=::NewEnumerator1, Loop Duration=1, UseLoopDelay=false)
   &SpawnBurst_Instantaneous(Use Spawn Probability=false, UseLoopCountLimit=false)
  ]
  particleSpawnScript [
   &InitializeParticle(
    Color=$User.ScreenColor
    Lifetime=0.8
    Mesh Scale=<1.2 1.35 1.35>
    Mesh Scale Mode=::NewEnumerator4
    UsePositionOffset=false
    Write Material Random=true
   )
   &InitialMeshOrientation(Mesh Orientation Mode=::NewEnumerator2, Rotation=<0 0.5 0>, Use Rotation Vector=true)
  ]
  particleUpdateScript [
   &ParticleState()
   &ScaleMeshSize(Scale Factor=&Multiply_VectorByFloat(Vector=<1 1 1>, Float=0.7))
   &ScaleColor(
    Scale Alpha=&FloatFromCurve(FloatCurve=DataInterface() {propertyValues {Curve {keys [
         {arriveTangent=0.7094, interpMode=RCIM_Cubic, time=0, value=0}
         {arriveTangent=-3.979, interpMode=RCIM_Cubic, leaveTangent=-3.979, tangentMode=RCTM_User, time=0.1, value=0.9591}
         {arriveTangent=-1.0657, interpMode=RCIM_Cubic, leaveTangent=-1.0657, tangentMode=RCTM_User, time=0.5031, value=0.2747}
         {arriveTangent=-1.1759, interpMode=RCIM_Cubic, leaveTangent=-1.1759, tangentMode=RCTM_User, time=1, value=0}
        ]}}})
    Scale RGB=&MakeVectorFromLinearColorRGB(Color=&ColorFromCurve(ColorCurve=DataInterface() {propertyValues {
        AlphaCurve {keys []}
        BlueCurve {keys [<at 0 0.7344 RCIM_Cubic>,<at 0.5487 1 RCIM_Cubic>]}
        GreenCurve {keys [<at 0 0.9974 RCIM_Cubic>,<at 0.5487 1 RCIM_Cubic>]}
        RedCurve {keys [<at 0 1 RCIM_Cubic>,<at 0.5487 1 RCIM_Cubic>]}
       }}))
    ScaleA=true
    ScaleRGB=true
   )
   &DynamicMaterialParameters(
    Index 0 Param 1=&FloatFromCurve(FloatCurve=DataInterface() {propertyValues {Curve {keys [{arriveTangent=-3.2102, interpMode=RCIM_Cubic, time=0, value=1.5}, <at 1 -0.2 RCIM_Cubic>]}}})
    Index 0 Param 2=&FloatFromCurve(FloatCurve=DataInterface() {propertyValues {Curve {keys [
         {arriveTangent=0.6385, interpMode=RCIM_Cubic, time=0, value=0.5}
         {arriveTangent=-0.1, interpMode=RCIM_Cubic, leaveTangent=-0.1, time=0.7948, value=0.75}
         {interpMode=RCIM_Cubic, leaveTangent=-0.0708, time=1, value=0.4}
        ]}}})
    Index 0 Param 3=1
    Index 0 Param 4=1
    Module.Index0Param0Write=true
    Module.Index0Param1Write=true
    Module.Index0Param2Write=true
    Module.Index0Param3Write=true
   )
  ]
  rendererClasses [&NiagaraMeshRendererProperties]
  renderers [{rendererClass=&NiagaraMeshRendererProperties, rendererIndex=0}]
 }
 Speedlines {
  simTarget=GPUComputeSim
  emitterUpdateScript [
   &EmitterState(Life Cycle Mode=::NewEnumerator1, Loop Behavior=::NewEnumerator1, Loop Duration=0.8, UseLoopDelay=false)
   &SpawnRate(
    SpawnRate=&Multiply_Float(
     A=500
     B=&FloatFromCurve(
      FloatCurve=DataInterface() {propertyValues {Curve {keys [
          {arriveTangent=7.8517, interpMode=RCIM_Cubic, leaveTangent=7.8517, tangentMode=RCTM_User, time=0, value=0.2}
          <at 0.2 1>
          {interpMode=RCIM_Cubic, tangentMode=RCTM_User, time=0.4, value=1}
          {arriveTangentWeight=0.8126, interpMode=RCIM_Cubic, tangentMode=RCTM_User, time=1, value=0.2}
         ]}}}
      CurveIndex=$Emitter.NormalizedLoopAge
     )
    )
    Use Spawn Probability=false
   )
  ]
  particleSpawnScript [
   /Niagara/Modules/Spawn/Initialization/InitializeParticle(
    Color=&LerpLinearColors(
     StartColor=<rgba 1 1 1 0.35>
     EndColor=<rgba 0 0 0 0.5>
     LerpFactor=&UniformAOrBFloat(A=0, Distribution Weight=0.15)
    )
    Lifetime=&UniformRangedFloat(Minimum=0.1, Maximum=0.15)
    Module.Write Color=true
    Module.Write Lifetime=true
    Module.Write Mass=true
    Module.Write Material Random=true
    Module.Write Position=false
    Module.Write Scale=false
    Module.Write SpriteRotation=false
    Module.Write SpriteSize=true
    Position=&SimulationPosition()
    Sprite Size=&UniformRangedVector2D(Minimum=<3 400>, Maximum=<1 500>)
   )
   &ShapeLocation(
    Handle Radius=0
    Large Radius=&Multiply_Float(
     A=&RandomRangeFloat(Minimum=50, Maximum=50)
     B=&FloatFromCurve(
      FloatCurve=DataInterface() {propertyValues {Curve {keys [
          {interpMode=RCIM_Cubic, tangentMode=RCTM_User, time=0, value=1}
          <at 0.2 0.4>
          {interpMode=RCIM_Cubic, tangentMode=RCTM_User, time=0.4, value=0.4}
          {arriveTangentWeight=0.8126, interpMode=RCIM_Cubic, tangentMode=RCTM_User, time=1, value=1}
         ]}}}
      CurveIndex=$Emitter.NormalizedLoopAge
     )
    )
    MaskSpawnGroup=false
    Non Uniform Scale=<1 1.7 1>
    Override Seed=false
    Rotation Mode=::NewEnumerator1
    Shape Primitive=::NewEnumerator3
    UseOwnerScale=true
   )
   &AddVelocityFromPoint(Module.Use Falloff Distance=false, Velocity Strength=&UniformRangedFloat(Minimum=1000, Maximum=300))
  ]
  particleUpdateScript [
   &ParticleState()
   &SolveForcesAndVelocity(Clamp Velocity=false, Limit Acceleration=false)
   &ScaleSpriteSize(Scale Factor=&Vector2DFromCurve(
     Vector2Curve=DataInterface() {propertyValues {XCurve {keys [<at 0 0.5 RCIM_Cubic>,<at 0.5 1 RCIM_Cubic>]}, YCurve {keys [<at 0 1 RCIM_Cubic>,<at 0.5 1 RCIM_Cubic>]}}}
     ScaleCurve=&Vector2DFromFloat(Value=1)
    ))
   &ScaleColor(
    Scale Alpha=&Multiply_Float(
     A=&FloatFromCurve(
      FloatCurve=DataInterface() {propertyValues {Curve {keys [
          {arriveTangent=5.2641, interpMode=RCIM_Cubic, leaveTangent=5.2641, tangentMode=RCTM_User, time=0, value=0}
          {arriveTangent=0.7132, interpMode=RCIM_Cubic, leaveTangent=0.7132, tangentMode=RCTM_User, time=0.3, value=1.0207}
          {arriveTangent=-2.9069, interpMode=RCIM_Cubic, leaveTangent=-2.9069, tangentMode=RCTM_User, time=1, value=0}
         ]}}}
      Scale Curve=&FloatFromCurve(
       FloatCurve=DataInterface() {propertyValues {Curve {keys [
           {interpMode=RCIM_Cubic, tangentMode=RCTM_User, time=0, value=0.1}
           {arriveTangent=-0.1, interpMode=RCIM_Cubic, leaveTangent=-0.1, time=0.1, value=1}
           {arriveTangent=-3.7503, interpMode=RCIM_Cubic, leaveTangent=-3.7503, tangentMode=RCTM_User, time=1, value=0}
          ]}}}
       CurveIndex=$Emitter.NormalizedLoopAge
      )
     )
     B=0.5
    )
    Scale RGB=&VectorFromFloat(Value=1)
    ScaleA=true
    ScaleRGB=true
   )
  ]
  rendererClasses [&NiagaraSpriteRendererProperties]
  renderers [{rendererClass=&NiagaraSpriteRendererProperties}]
 }
}
summary {
 systemName=NS_Masamune_Camera_Lens_Sprint
 emitters [
  {emitterName=ScreenWave, rendererClasses [&NiagaraMeshRendererProperties]}
  {emitterName=Speedlines, simTarget=GPUComputeSim, rendererClasses [&NiagaraSpriteRendererProperties]}
 ]
 userVariables [
  User.LifeTime {type=NiagaraFloat, defaultValue=1, description=""}
  User.ScreenColor {type=LinearColor, description="", defaultValue {struct=LinearColor, value=<rgba 1 1 1 0.8>}}
 ]
}
