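# Niagara one-shot slash effect, named for Masamune's old wandering boss, built from layered meshes and sprite streaks.
# CombineMesh packs six layers into one emitter, switching scale, color and fade by spawn group.
# Flow and Wind add meshes sharing system position, velocity and drag; a particle emitter throws curl-noise sprites.
# User Color Variant blends pink-red toward green palettes in CombineMesh and Flow; system variables tint the other emitters.
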NiagaraSystem /MasamuneGameplay/Assets/FX/NewFX/NS_Masamune_OldWangderingBossSlash
dependencies {
 usedDataInterfaces [
  &NiagaraDataInterfaceColorCurve
  &NiagaraDataInterfaceCurve
  &NiagaraDataInterfaceVectorCurve
  &NiagaraDataInterfaceArrayColor
 ]
 usedDynamicInputs [&SelectLinearColorFromArray]
 usedModules [
  &self:SystemSpawnScript.SystemScriptSource.SystemScriptGraph.NiagaraNodeAssignment_1.SetVariables_#1
  &self:SystemSpawnScript.SystemScriptSource.SystemScriptGraph.NiagaraNodeAssignment_2.SetVariables_#2
  &self:SystemSpawnScript.SystemScriptSource.SystemScriptGraph.NiagaraNodeAssignment_4.SetVariables_#3
  &self:SystemSpawnScript.SystemScriptSource.SystemScriptGraph.NiagaraNodeAssignment_3.SetVariables_#4
  &self:SystemSpawnScript.SystemScriptSource.SystemScriptGraph.NiagaraNodeAssignment_0.SetVariables_#5
  &SystemState
 ]
}
emitters {
 CombineMesh {
  emitterUpdateScript [
   &EmitterState(Life Cycle Mode=::NewEnumerator1, Loop Behavior=::NewEnumerator1, Loop Duration=1, UseLoopDelay=false)
   &SpawnBurst_Instantaneous(Use Spawn Probability=false, UseLoopCountLimit=false)
   &SpawnBurst_Instantaneous(Spawn Group=1, Use Spawn Probability=false, UseLoopCountLimit=false) {module=SpawnBurst_Instantaneous001}
   &SpawnBurst_Instantaneous(Spawn Group=2, Use Spawn Probability=false, UseLoopCountLimit=false) {module=SpawnBurst_Instantaneous002}
   &SpawnBurst_Instantaneous(Spawn Group=3, Use Spawn Probability=false, UseLoopCountLimit=false) {module=SpawnBurst_Instantaneous003}
   &SpawnBurst_Instantaneous(Spawn Group=4, Use Spawn Probability=false, UseLoopCountLimit=false) {module=SpawnBurst_Instantaneous004}
   &SpawnBurst_Instantaneous(Spawn Group=5, Use Spawn Probability=false, UseLoopCountLimit=false) {module=SpawnBurst_Instantaneous005}
  ]
  particleSpawnScript [
   &InitializeParticle(
    Color=&LerpLinearColors(
     StartColor=&MergeHelper_ColorSwitch(
      0=<rgba 3 3 3 1>
      1=<rgba 4 4 4 1>
      2=<rgba 0.3 0.3 0.3 2>
      3=<rgba 1 1 1 1>
      4=<rgba 1 1 1 1>
      5=<rgba 0.3177 0.0886 0.1268 4>
      6=<rgba 0 0 0 0>
      7=<rgba 0 0 0 0>
      8=<rgba 0 0 0 0>
      9=<rgba 0 0 0 0>
      Index=&SpawnGroup()
      Num=::NewEnumerator8
     )
     EndColor=&MergeHelper_ColorSwitch(
      0=<rgba 0.1794 3 1.2094 1>
      1=<rgba 0.863 4 0.5637 1>
      2=<rgba 0.0678 0.3 0.0293 2>
      3=<rgba 0.164 0.4882 1 1>
      4=<rgba 0.3493 1 0.5255 1>
      5=<rgba 0.0147 0.3177 0.0425 4>
      6=<rgba 0 0 0 0>
      7=<rgba 0 0 0 0>
      8=<rgba 0 0 0 0>
      9=<rgba 0 0 0 0>
      Index=&SpawnGroup()
      Num=::NewEnumerator8
     )
     LerpFactor=&MakeFloatFromInt(Int=$User.Color Variant)
    )
    Lifetime=&MergeHelper_FloatSwitch(0=0.5, 1=0.4, 2=0.5, 3=0.7, 4=0.5, 5=0.6, 6=0, 7=0, 8=0, 9=0, Index=&SpawnGroup(), Num=::NewEnumerator8)
    Mesh Scale=&MergeHelper_VectorSwitch(
     0=&Multiply_VectorByFloat(Vector=<1 1 0.1>, Float=1.85)
     1=&Multiply_VectorByFloat(Vector=<1 0.92 0.5>, Float=1.2)
     2=&Multiply_VectorByFloat(Vector=<1 1 0.4>, Float=1.9)
     3=&Multiply_VectorByFloat(Vector=<1 1 0.1>, Float=1.5)
     4=&Multiply_VectorByFloat(Vector=<1 0.92 0.2>, Float=1.1)
     5=&Multiply_VectorByFloat(Vector=<1 1 0.3>, Float=1.5)
     6=<0 0 0>
     7=<0 0 0>
     8=<0 0 0>
     9=<0 0 0>
     Index=&SpawnGroup()
     Num=::NewEnumerator8
    )
    Mesh Scale Mode=::NewEnumerator4
    Position Offset=&MergeHelper_VectorSwitch(
     0=&AddVectorToPosition(Position=<0 0 0>, Vector=$System.FXPosition)
     1=&AddVectorToPosition(Position=<-10 1 3>, Vector=$System.FXPosition)
     2=&AddVectorToPosition(Position=<0 0 0>, Vector=$System.FXPosition)
     3=&AddVectorToPosition(Position=<-100 0 5>, Vector=$System.FXPosition)
     4=&AddVectorToPosition(Position=<8 1 3>, Vector=$System.FXPosition)
     5=&AddVectorToPosition(Position=<-10 1 3>, Vector=$System.FXPosition)
     6=<0 0 0>
     7=<0 0 0>
     8=<0 0 0>
     9=<0 0 0>
     Index=&SpawnGroup()
     Num=::NewEnumerator8
    )
    UsePositionOffset=true
    Write Material Random=true
   )
   &InitialMeshOrientation(Mesh Orientation Mode=::NewEnumerator7, Rotation=<0 0 0.5>, Use Rotation Vector=true) {module=InitialMeshOrientation001}
   &InitialMeshOrientation(Mesh Orientation Mode=::NewEnumerator7, Use Rotation Vector=true)
   &AddVelocity(Velocity=$System.FXvelocity)
   /SetVariables(Particles.VisibilityTag=&SpawnGroup()) {isSetParameters=true}
  ]
  particleUpdateScript [
   &ParticleState()
   &ScaleMeshSize(Scale Factor=&Multiply_Vector(A=$Engine.Owner.Scale, B=$System.SizeScale))
   &ScaleColor(
    Scale Alpha=&MergeHelper_FloatSwitch(
     0=&Multiply_Float(
      A=10
      B=&FloatFromCurve(FloatCurve=DataInterface() {propertyValues {Curve {keys [<at 0 1>,<at 0.7 1>,<at 1 0>]}}}, Scale Curve=2)
     )
     1=&Multiply_Float(
      A=0.7
      B=&FloatFromCurve(FloatCurve=DataInterface() {propertyValues {Curve {keys [
           <at 0 0 RCIM_Cubic>
           {arriveTangent=1.4583, interpMode=RCIM_Cubic, leaveTangent=1.4583, time=0.5, value=0.925}
           {arriveTangent=-1.85, interpMode=RCIM_Cubic, leaveTangent=-1.85, time=0.6, value=0.875}
           <at 1 0 RCIM_Cubic>
          ]}}})
     )
     2=&FloatFromCurve(
      FloatCurve=DataInterface() {propertyValues {Curve {keys [<at 0 0>, {arriveTangent=-1.4282, interpMode=RCIM_Cubic, leaveTangent=-1.4282, tangentMode=RCTM_User, time=0.6, value=0.8}, <at 1 0>]}}}
      Scale Curve=2
     )
     3=&FloatFromCurve(
      FloatCurve=DataInterface() {propertyValues {Curve {keys [
          <at 0 1 RCIM_Cubic>
          {arriveTangent=-0.8307, interpMode=RCIM_Cubic, leaveTangent=-0.8307, time=0.4812, value=0.7644}
          {arriveTangent=-4.2795, interpMode=RCIM_Cubic, leaveTangent=-4.2795, tangentMode=RCTM_User, time=0.5844, value=0.5146}
          {arriveTangent=-1.1825, interpMode=RCIM_Cubic, leaveTangent=-1.1825, tangentMode=RCTM_User, time=1, value=0}
         ]}}}
      Scale Curve=2
     )
     4=&FloatFromCurve(
      FloatCurve=DataInterface() {propertyValues {Curve {keys [<at 0 1 RCIM_Cubic>, {arriveTangent=-1.5335, tangentMode=RCTM_User, time=1, value=0}]}}}
      Scale Curve=0.5
     )
     5=&Multiply_Float(
      A=0.2
      B=&FloatFromCurve(FloatCurve=DataInterface() {propertyValues {Curve {keys [
           <at 0 0 RCIM_Cubic>
           {arriveTangent=0.8571, interpMode=RCIM_Cubic, leaveTangent=0.8571, time=0.2, value=1}
           {arriveTangent=-1.25, interpMode=RCIM_Cubic, leaveTangent=-1.25, time=0.7, value=0.6}
           <at 1 0 RCIM_Cubic>
          ]}}})
     )
     6=0
     7=0
     8=0
     9=0
     Index=&SpawnGroup()
     Num=::NewEnumerator8
    )
    Scale RGB=&MergeHelper_VectorSwitch(
     0=&MakeVectorFromLinearColorRGB(Color=&LerpLinearColors(
       StartColor=&ColorFromCurve(ColorCurve=DataInterface() {propertyValues {
          AlphaCurve {keys []}
          BlueCurve {keys [
            <at 0.0056 2.875>
            <at 0.1845 4>
            {arriveTangent=-8.807, interpMode=RCIM_Cubic, leaveTangent=-8.807, time=0.3426, value=2.4467}
            <at 0.5519 0.7643 RCIM_Cubic>
           ]}
          GreenCurve {keys [
            <at 0.0056 0.5>
            <at 0.1845 1>
            {arriveTangent=-1.9243, interpMode=RCIM_Cubic, leaveTangent=-1.9243, time=0.3426, value=1.17}
            <at 0.5519 0.293 RCIM_Cubic>
           ]}
          RedCurve {keys [
            <at 0.0056 10>
            <at 0.1845 10>
            {arriveTangent=-24.4966, interpMode=RCIM_Cubic, leaveTangent=-24.4966, time=0.3426, value=5}
            <at 0.5519 1 RCIM_Cubic>
           ]}
         }})
       EndColor=&ColorFromCurve(ColorCurve=DataInterface() {propertyValues {
          AlphaCurve {keys []}
          BlueCurve {keys [
            <at 0.0056 4.1625>
            <at 0.1845 3.7183>
            {arriveTangent=-5.9197, interpMode=RCIM_Cubic, leaveTangent=-5.9197, time=0.3562, value=0.756}
            <at 0.7832 0.1741 RCIM_Cubic>
           ]}
          GreenCurve {keys [
            <at 0.0056 10>
            <at 0.1845 10>
            {arriveTangent=-15.0322, interpMode=RCIM_Cubic, leaveTangent=-15.0322, time=0.3562, value=2}
            <at 0.7832 1 RCIM_Cubic>
           ]}
          RedCurve {keys [
            <at 0.0056 1.4959>
            <at 0.1845 5.4128>
            {arriveTangent=-8.8688, interpMode=RCIM_Cubic, leaveTangent=-8.8688, time=0.3562, value=1.2824}
            <at 0.7832 0.1029 RCIM_Cubic>
           ]}
         }})
       LerpFactor=&MakeFloatFromInt(Int=$User.Color Variant)
      ))
     1=&MakeVectorFromLinearColorRGB(Color=/Niagara/DynamicInputs/Lerp/SmoothLerpOverTime_Color(
       Smooth Value:LinearColor=<rgba 0.7 0.1029 0.2522 1>
       Target Value:LinearColor=<rgba 0.0392 1 0.326 1>
       Convergence Rate:NiagaraFloat=&MakeFloatFromInt(Int=$User.Color Variant)
      ))
     2=&Multiply_VectorByFloat(
      Vector=&MakeVectorFromLinearColorRGB(Color=&SmoothLerpOverTime_Color(
        Smooth Value:LinearColor=&ColorFromCurve(ColorCurve=DataInterface() {propertyValues {
           AlphaCurve {keys []}
           BlueCurve {keys [<at 0 0.8368 RCIM_Cubic>, {arriveTangent=-0.064, interpMode=RCIM_Cubic, leaveTangent=-0.064, time=0.3631, value=0.5}, <at 0.6778 0.7934>]}
           GreenCurve {keys [<at 0 0.6042 RCIM_Cubic>, {arriveTangent=-0.0768, interpMode=RCIM_Cubic, leaveTangent=-0.0768, time=0.3631, value=0.2}, <at 0.6778 0.5521>]}
           RedCurve {keys [<at 0 2 RCIM_Cubic>,<at 0.3631 2 RCIM_Cubic>,<at 0.6778 2>]}
          }})
        Target Value:LinearColor=&ColorFromCurve(ColorCurve=DataInterface() {propertyValues {
           AlphaCurve {keys []}
           BlueCurve {keys [<at 0 1.0482 RCIM_Cubic>, {interpMode=RCIM_Cubic, leaveTangent=-1.3305, time=0.3631, value=0.5909}, <at 0.9168 0.1463>]}
           GreenCurve {keys [<at 0 2 RCIM_Cubic>,<at 0.3631 2 RCIM_Cubic>,<at 0.9168 2>]}
           RedCurve {keys [<at 0 0.5506 RCIM_Cubic>, {interpMode=RCIM_Cubic, leaveTangent=-0.8123, time=0.3631, value=0.1752}, <at 0.9168 0>]}
          }})
        Convergence Rate:NiagaraFloat=&MakeFloatFromInt(Int=$User.Color Variant)
       ))
      Float=&FloatFromCurve(
       FloatCurve=DataInterface() {propertyValues {Curve {keys [{arriveTangent=-0.0133, interpMode=RCIM_Cubic, leaveTangent=-0.0133, tangentMode=RCTM_User, time=0, value=0.5}, <at 0.5 0.5>]}}}
       Scale Curve=5
      )
     )
     3=&Multiply_VectorByFloat(
      Vector=&Lerp_Vector(
       A=&MakeVectorFromLinearColorRGB(Color=&ColorFromCurve(ColorCurve=DataInterface() {propertyValues {
           AlphaCurve {keys []}
           BlueCurve {keys [<at 0 0.75 RCIM_Cubic>, {arriveTangent=-0.4167, interpMode=RCIM_Cubic, leaveTangent=-0.4167, time=0.3549, value=0.75}, <at 1 0.3333>]}
           GreenCurve {keys [<at 0 0.3 RCIM_Cubic>, {arriveTangent=-0.1, interpMode=RCIM_Cubic, leaveTangent=-0.1, time=0.3549, value=0.3}, <at 1 0.2>]}
           RedCurve {keys [<at 0 3 RCIM_Cubic>, {arriveTangent=-2, interpMode=RCIM_Cubic, leaveTangent=-2, time=0.3549, value=3}, <at 1 1>]}
          }}))
       B=&MakeVectorFromLinearColorRGB(Color=&ColorFromCurve(ColorCurve=DataInterface() {propertyValues {
           AlphaCurve {keys []}
           BlueCurve {keys [<at 0 3 RCIM_Cubic>, {arriveTangent=-2.8927, interpMode=RCIM_Cubic, leaveTangent=-2.8927, time=0.3549, value=1.0582}, <at 1 0.1073>]}
           GreenCurve {keys [<at 0 1.9268 RCIM_Cubic>, {arriveTangent=-0.9268, interpMode=RCIM_Cubic, leaveTangent=-0.9268, time=0.3549, value=3}, <at 1 1>]}
           RedCurve {keys [<at 0 0.6918 RCIM_Cubic>, {arriveTangent=-0.3307, interpMode=RCIM_Cubic, leaveTangent=-0.3307, time=0.3549, value=0.1947}, <at 1 0.361>]}
          }}))
       Alpha=&MakeFloatFromInt(Int=$User.Color Variant)
      )
      Float=&FloatFromCurve(FloatCurve=DataInterface() {propertyValues {Curve {keys [<at 0 1>,<at 0.5 1>]}}}, Scale Curve=3)
     )
     4=&VectorFromCurve(VectorCurve=DataInterface() {propertyValues {
        XCurve {keys [<at 0 1.9>,<at 1 1>]}
        YCurve {keys [{arriveTangent=-0.4825, interpMode=RCIM_Cubic, leaveTangent=-0.4825, tangentMode=RCTM_User, time=0, value=0.8}, <at 1 0.575>]}
        ZCurve {keys [<at 0 0.6>,<at 1 0.5>]}
       }})
     5=<1 1 1>
     6=<0 0 0>
     7=<0 0 0>
     8=<0 0 0>
     9=<0 0 0>
     Index=&SpawnGroup()
     Num=::NewEnumerator8
    )
    ScaleA=true
    ScaleRGB=true
   )
   &DynamicMaterialParameters(
    Index 0 Param 1=&MergeHelper_FloatSwitch(
     0=&FloatFromCurve(FloatCurve=DataInterface() {propertyValues {Curve {keys [<at 0 0 RCIM_Cubic>,<at 1 1 RCIM_Cubic>]}}})
     1=&FloatFromCurve(FloatCurve=DataInterface() {propertyValues {Curve {keys [<at 0 0 RCIM_Cubic>,<at 1 0 RCIM_Cubic>]}}})
     2=&FloatFromCurve(FloatCurve=DataInterface() {propertyValues {Curve {keys [<at 0 0 RCIM_Cubic>,<at 1 0>]}}})
     3=&FloatFromCurve(FloatCurve=DataInterface() {propertyValues {Curve {keys [<at 0 0 RCIM_Cubic>,<at 1 1>]}}})
     4=&FloatFromCurve(FloatCurve=DataInterface() {propertyValues {Curve {keys [<at 0 0 RCIM_Cubic>,<at 1 0 RCIM_Cubic>]}}})
     5=&FloatFromCurve(FloatCurve=DataInterface() {propertyValues {Curve {keys [<at 0 0 RCIM_Cubic>,<at 1 0 RCIM_Cubic>]}}})
     6=0
     7=0
     8=0
     9=0
     Index=&SpawnGroup()
     Num=::NewEnumerator8
    )
    Index 0 Param 2=&MergeHelper_FloatSwitch(0=0.6, 1=0.6, 2=0.1, 3=0.1, 4=0.6, 5=0.6, 6=0, 7=0, 8=0, 9=0, Index=&SpawnGroup(), Num=::NewEnumerator8)
    Module.Index0Param0Write=true
    Module.Index0Param1Write=true
    Module.Index0Param2Write=true
    Module.Index0Param3Write=true
   )
   &Drag(Drag=$System.VelocityDrag, Use Linear Drag=true, Use Rotational Drag=true)
   &SolveForcesAndVelocity(Clamp Velocity=false, Limit Acceleration=false)
  ]
  rendererClasses [&NiagaraMeshRendererProperties]
  renderers [
   {rendererClass=&NiagaraMeshRendererProperties, rendererIndex=0}
   {rendererClass=&NiagaraMeshRendererProperties, rendererIndex=1}
   {rendererClass=&NiagaraMeshRendererProperties, rendererIndex=2}
   {rendererClass=&NiagaraMeshRendererProperties, rendererIndex=3}
   {rendererClass=&NiagaraMeshRendererProperties, rendererIndex=4}
   {rendererClass=&NiagaraMeshRendererProperties, rendererIndex=5}
  ]
 }
 Flow {
  emitterUpdateScript [
   &EmitterState(Life Cycle Mode=::NewEnumerator1, Loop Behavior=::NewEnumerator1, Loop Duration=1, UseLoopDelay=true)
   &SpawnBurst_Instantaneous(Use Spawn Probability=false, UseLoopCountLimit=false)
  ]
  particleSpawnScript [
   &InitializeParticle(
    Color=$System.color_a
    Lifetime=0.6
    Mesh Scale=&Multiply_VectorByFloat(Vector=<1 1 1>, Float=1.3)
    Mesh Scale Mode=::NewEnumerator4
    Position Offset=&AddVectorToPosition(Position=<400 0 -10>, Vector=$System.FXPosition)
    UsePositionOffset=true
    Write Material Random=true
   )
   &InitialMeshOrientation(Mesh Orientation Mode=::NewEnumerator7, Rotation=<0 0 0.5>, Use Rotation Vector=true)
   &InitialMeshOrientation(Mesh Orientation Mode=::NewEnumerator7, Use Rotation Vector=true) {module=InitialMeshOrientation001}
   &AddVelocity(Velocity=$System.FXvelocity)
  ]
  particleUpdateScript [
   &ParticleState()
   &ScaleMeshSize(Scale Factor=&Multiply_Vector(
     A=&VectorFromCurve(
      VectorCurve=DataInterface() {propertyValues {XCurve {keys [<at 0 1>,<at 1 1>]}, YCurve {keys [<at 0 1>,<at 1 1>]}, ZCurve {keys [<at 0 1>,<at 1 1>]}}}
      Scale Curve=<1.2 1 1>
     )
     B=&Multiply_Vector(A=$Engine.Owner.Scale, B=$System.SizeScale)
    )) {module=ScaleMeshSize001}
   &ScaleColor(
    Scale Alpha=&FloatFromCurve(
     FloatCurve=DataInterface() {propertyValues {Curve {keys [
         <at 0 0 RCIM_Cubic>
         {arriveTangent=0.4858, interpMode=RCIM_Cubic, leaveTangent=0.4858, time=0.3, value=0.9794}
         {arriveTangent=-1.9358, interpMode=RCIM_Cubic, leaveTangent=-1.9358, tangentMode=RCTM_User, time=0.5749, value=0.2793}
         {arriveTangent=-0.6689, interpMode=RCIM_Cubic, leaveTangent=-0.6689, tangentMode=RCTM_User, time=1, value=0}
        ]}}}
     Scale Curve=1.5
    )
    Scale RGB=&MakeVectorFromLinearColorRGB(Color=&SmoothLerpOverTime_Color(
      Smooth Value:LinearColor=&ColorFromCurve(ColorCurve=DataInterface() {propertyValues {
         AlphaCurve {keys [<at 0 1 RCIM_Cubic>]}
         BlueCurve {keys [<at 0 0.5 RCIM_Cubic>,<at 1 0.6667 RCIM_Cubic>]}
         GreenCurve {keys [<at 0 0.2 RCIM_Cubic>,<at 1 0.4 RCIM_Cubic>]}
         RedCurve {keys [<at 0 2 RCIM_Cubic>,<at 1 2 RCIM_Cubic>]}
        }})
      Target Value:LinearColor=&ColorFromCurve(ColorCurve=DataInterface() {propertyValues {
         AlphaCurve {keys [<at 0 1 RCIM_Cubic>]}
         BlueCurve {keys [<at 0 0.3641 RCIM_Cubic>,<at 1 0.0009 RCIM_Cubic>]}
         GreenCurve {keys [<at 0 2 RCIM_Cubic>,<at 1 2 RCIM_Cubic>]}
         RedCurve {keys [<at 0 0.144 RCIM_Cubic>,<at 1 0.0551 RCIM_Cubic>]}
        }})
      Convergence Rate:NiagaraFloat=&MakeFloatFromInt(Int=$User.Color Variant)
     ))
    ScaleA=true
    ScaleRGB=true
   )
   &DynamicMaterialParameters(
    Index 0 Param 1=2
    Index 0 Param 2=100
    Module.Index0Param0Write=true
    Module.Index0Param1Write=true
    Module.Index0Param2Write=true
    Module.Index0Param3Write=true
   )
   &FN_Mesh_Attributes(
    Color=&MakeLinearColorFromVectorAndFloat(
     "Vector (RGB)"=&MakeVectorFromLinearColorRGB(Color=$System.color_b)
     "Float (Alpha)"=&FloatFromCurve(
      FloatCurve=DataInterface() {propertyValues {Curve {keys [
          <at 0 0 RCIM_Cubic>
          {arriveTangent=0.4858, interpMode=RCIM_Cubic, leaveTangent=0.4858, time=0.2, value=0.9794}
          {arriveTangent=-1.9358, interpMode=RCIM_Cubic, leaveTangent=-1.9358, tangentMode=RCTM_User, time=0.5749, value=0.2793}
          {arriveTangent=-0.6689, interpMode=RCIM_Cubic, leaveTangent=-0.6689, tangentMode=RCTM_User, time=1, value=0}
         ]}}}
      Scale Curve=0.5
     )
    )
    Create Color Attribute=true
    Create Scale Attribute=true
    Scale=&Multiply_Vector(A=<1.2 1 1>, B=&Multiply_VectorByFloat(Vector=<1 1 1>, Float=8))
    Write to Module Namespace=true
   )
   &Drag(Drag=$System.VelocityDrag, Use Linear Drag=true, Use Rotational Drag=true)
   &SolveForcesAndVelocity(Clamp Velocity=false, Limit Acceleration=false)
  ]
  rendererClasses [&NiagaraMeshRendererProperties]
  renderers [
   {rendererClass=&NiagaraMeshRendererProperties, rendererIndex=0}
   {rendererClass=&NiagaraMeshRendererProperties, rendererIndex=1}
  ]
 }
 Wind {
  emitterUpdateScript [
   &EmitterState(Life Cycle Mode=::NewEnumerator1, Loop Behavior=::NewEnumerator1, Loop Duration=1.5, UseLoopDelay=true)
   &SpawnBurst_Instantaneous(Use Spawn Probability=false, UseLoopCountLimit=false)
   &SpawnBurst_Instantaneous(Spawn Time=0.1, Use Spawn Probability=false, UseLoopCountLimit=false) {enabled=false, module=SpawnBurst_Instantaneous001}
  ]
  particleSpawnScript [
   &InitializeParticle(
    Color=$System.color_c
    Lifetime=0.9
    Mesh Scale=&Multiply_VectorByFloat(Vector=<1 0.92 0.2>, Float=2.3)
    Mesh Scale Mode=::NewEnumerator4
    Position Offset=&AddVectorToPosition(Position=<100 1 3>, Vector=$System.FXPosition)
    UsePositionOffset=true
    Write Material Random=true
   )
   &InitialMeshOrientation(Mesh Orientation Mode=::NewEnumerator7, Rotation=<0 0 0.5>, Use Rotation Vector=true)
   &InitialMeshOrientation(Mesh Orientation Mode=::NewEnumerator7, Use Rotation Vector=true) {module=InitialMeshOrientation001}
   &AddVelocity(Velocity=$System.FXvelocity)
  ]
  particleUpdateScript [
   &Drag(Drag=$System.VelocityDrag, Use Linear Drag=true, Use Rotational Drag=true)
   &SolveForcesAndVelocity(Clamp Velocity=false, Limit Acceleration=false)
   &ParticleState()
   &UpdateMeshOrientation(
    Rotation Rate=&FloatFromCurve(FloatCurve=DataInterface() {propertyValues {Curve {keys [{arriveTangent=-1.6177, interpMode=RCIM_Cubic, leaveTangent=-1.6177, tangentMode=RCTM_User, time=0, value=1}, <at 1 0.1>]}}})
    Rotation Vector=<0 0 1>
   ) {enabled=false}
   &ScaleMeshSize(Scale Factor=&Multiply_Vector(A=$Engine.Owner.Scale, B=$System.SizeScale))
   &ScaleMeshSize(Scale Factor=&Multiply_VectorByFloat(
     Vector=<1 1 1.5>
     Float=&FloatFromCurve(FloatCurve=DataInterface() {propertyValues {Curve {keys [<at 0 0.7>,<at 1 0.8>]}}})
    )) {module=ScaleMeshSize001}
   &ScaleColor(
    Scale Alpha=&Multiply_Float(
     A=1
     B=&FloatFromCurve(
      FloatCurve=DataInterface() {propertyValues {Curve {keys [
          <at 0 0 RCIM_Cubic>
          {arriveTangent=1.3571, interpMode=RCIM_Cubic, leaveTangent=1.3571, time=0.1, value=1.6201}
          {arriveTangent=-1.8001, interpMode=RCIM_Cubic, leaveTangent=-1.8001, time=0.7, value=0.95}
          <at 1 0 RCIM_Cubic>
         ]}}}
      Scale Curve=0.2
     )
    )
    Scale RGB=&MakeVectorFromLinearColorRGB(Color=<rgba 0.5 0.5 0.5 1>)
    ScaleA=true
    ScaleRGB=true
   )
   &DynamicMaterialParameters(
    Index 0 Param 1=&FloatFromCurve(FloatCurve=DataInterface() {propertyValues {Curve {keys [{interpMode=RCIM_Cubic, tangentMode=RCTM_User, time=0.5, value=0}, {interpMode=RCIM_Cubic, tangentMode=RCTM_User, time=1, value=1}]}}})
    Index 0 Param 3=1
    Module.Param0WriteEnabled=true
    Module.Param1WriteEnabled=true
    Module.Param2WriteEnabled=true
    Module.Param3WriteEnabled=true
   ) {module=DynamicMaterialParameters001}
  ]
  rendererClasses [&NiagaraMeshRendererProperties]
  renderers [{rendererClass=&NiagaraMeshRendererProperties, rendererIndex=0}]
 }
 particle {
  emitterUpdateScript [
   &EmitterState(Life Cycle Mode=::NewEnumerator1, Loop Behavior=::NewEnumerator1, Loop Duration=0.3, UseLoopDelay=false)
   &SpawnBurst_Instantaneous(Spawn Count=10, Use Spawn Probability=false, UseLoopCountLimit=false)
   &SpawnRate(SpawnRate=80, Use Spawn Probability=false)
  ]
  particleSpawnScript [
   &InitializeParticle(
    Color=$System.color_d
    Lifetime Max=0.5
    Lifetime Min=0.3
    Lifetime Mode=::NewEnumerator1
    Sprite Size Max=<0.5 7>
    Sprite Size Min=<0.5 5>
    Sprite Size Mode=::NewEnumerator2
    UsePositionOffset=false
    Write Material Random=true
   )
   &ShapeLocation(
    Box Size=<100 600 1>
    MaskSpawnGroup=false
    Offset=<500 0 100>
    Offset Mode=::NewEnumerator0
    Override Seed=false
    Shape Primitive=::NewEnumerator2
    Surface Only=false
    UseOwnerScale=true
   )
   &AddVelocity(Velocity=<-5000 0 0>)
   &AddVelocity(Module.Use Falloff Distance=false, Origin Offset=<0 0 100>, Override Seed=false, Velocity Mode=::NewEnumerator1, Velocity Speed=2000) {module=AddVelocity001}
  ]
  particleUpdateScript [
   &ParticleState()
   &ScaleSpriteSize(Scale Sprite Size Mode=::NewEnumerator0, Uniform Scale Factor=2)
   &CurlNoiseForce(Bias Noise=false, Noise Strength=3000, Pan Noise=false)
   &Drag(
    Drag=&FloatFromCurve(FloatCurve=DataInterface() {propertyValues {Curve {keys [<at 0 1>,<at 1 0>]}}}, Scale Curve=5)
    Use Linear Drag=true
    Use Rotational Drag=true
   )
   &SolveForcesAndVelocity(Clamp Velocity=false, Limit Acceleration=false)
   &ScaleSpriteSizeBySpeed(Max Scale Factor=<1 2>, Min Scale Factor=<0.5 0.5>)
   &FN_Sprite_Attributes(
    Color=$System.color_e
    Create Color Attribute=true
    Create SpriteSize Attribute=true
    SpriteSize=&Multiply_Vector2D(A=$Particles.SpriteSize, B=<30 2>)
    Write to Module Namespace=true
   )
   &DynamicMaterialParameters(
    Index 0 Param 1=2
    Module.Index0Param0Write=true
    Module.Index0Param1Write=true
    Module.Index0Param2Write=true
    Module.Index0Param3Write=true
   )
   /SetVariables(Particles.Position02:NiagaraPosition=&AddVectorToPosition(
     Position=$Particles.Position
     Vector=&MakeVector(X=&Multiply_Float(A=$Emitter.Age, B=1000), Y=0, Z=0)
    )) {isSetParameters=true}
  ]
  rendererClasses [&NiagaraSpriteRendererProperties]
  renderers [
   {rendererClass=&NiagaraSpriteRendererProperties}
   {rendererClass=&NiagaraSpriteRendererProperties, rendererIndex=1}
  ]
 }
}
summary {
 systemName=NS_Masamune_OldWangderingBossSlash
 emitters [
  {emitterName=CombineMesh, rendererClasses [&NiagaraMeshRendererProperties]}
  {emitterName=Wind, rendererClasses [&NiagaraMeshRendererProperties]}
  {emitterName=Flow, rendererClasses [&NiagaraMeshRendererProperties]}
  {emitterName=particle, rendererClasses [&NiagaraSpriteRendererProperties]}
 ]
 userVariables [
  User.bIsFrontEnd {type=NiagaraBool}
  User.bIsFrontEndPreview {type=NiagaraBool}
  User.Color Variant {type=NiagaraInt32}
  User.FXPositionOffset {type=Vector3f, description="", defaultValue {struct=Vector3f, value=<0 0 0>}}
  User.FXvelocityOffset {type=Vector3f, description="", defaultValue {struct=Vector3f, value=<0 0 0>}}
  User.IsFrontEnd {type=NiagaraBool}
 ]
}
