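# Niagara system for the Masamune sword swing trail: a short-lived, animation-driven ribbon effect.
# Its single Trail emitter uses custom AnimTrail modules to spawn, initialize bone indices and update particles from a system-level skeletal mesh.
# Curves shift the ribbon from bright warm yellow to neutral grey and fade its alpha, with a separate fading yellow color written by FN_Sprite_Attributes.
# It exposes a RibbonIDOffset user parameter.
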NiagaraSystem /MasamuneGameplay/Assets/FX/NewFX/NS_Masamune_SwordAnimTrail_Swing
dependencies {
 usedDataInterfaces [&NiagaraDataInterfaceColorCurve, &NiagaraDataInterfaceCurve, &NiagaraDataInterfaceSkeletalMesh]
 usedDynamicInputs []
 usedModules [
  &self:SystemSpawnScript.SystemScriptSource.SystemScriptGraph.NiagaraNodeAssignment_0.SetVariables_#1
  &SystemState
 ]
}
emitters {Trail {
  emitterUpdateScript [
   &EmitterState()
   &FN_SpawnAnimTrail()
  ]
  particleSpawnScript [
   &FN_InitializeAnimTrailBoneIndecies()
   &FN_InitializeAnimTrail(Color=<rgba 3 3 3 1>, Lifetime=0.2, SkeletalMesh=$System.Skeletal Mesh)
   &AddVelocity(Velocity=<0 0 0>)
  ]
  particleUpdateScript [
   &SolveForcesAndVelocity(Clamp Velocity=false, Limit Acceleration=false)
   &ParticleState()
   &FN_UpdateAnimTrail(SkeletalMesh=$System.Skeletal Mesh)
   &ScaleColor(
    Scale Alpha=&FloatFromCurve(FloatCurve=DataInterface() {propertyValues {Curve {keys [{arriveTangent=-4.9584, interpMode=RCIM_Cubic, leaveTangent=-4.9584, tangentMode=RCTM_User, time=0.3, value=1}, <at 0.9 0 RCIM_Cubic>]}}})
    Scale RGB=&MakeVectorFromLinearColorRGB(Color=&ColorFromCurve(ColorCurve=DataInterface() {propertyValues {
        AlphaCurve {keys []}
        BlueCurve {keys [<at 0 0.5776 RCIM_Cubic>,<at 0.1555 0.4844 RCIM_Cubic>]}
        GreenCurve {keys [<at 0 1.861 RCIM_Cubic>,<at 0.1555 0.4844 RCIM_Cubic>]}
        RedCurve {keys [<at 0 2 RCIM_Cubic>,<at 0.1555 0.4844 RCIM_Cubic>]}
       }}))
    ScaleA=true
    ScaleRGB=true
   )
   &FN_Sprite_Attributes(
    Color=&MakeLinearColorFromVectorAndFloat(
     "Vector (RGB)"=&Multiply_Vector(
      A=&MakeVectorFromLinearColorRGB(Color=&ColorFromCurve(ColorCurve=DataInterface() {propertyValues {
          AlphaCurve {keys []}
          BlueCurve {keys [<at 0 0.2417 RCIM_Cubic>]}
          GreenCurve {keys [<at 0 1 RCIM_Cubic>]}
          RedCurve {keys [<at 0 1 RCIM_Cubic>]}
         }}))
      B=&Multiply_VectorByFloat(Vector=<1 1 1>, Float=0.4)
     )
     "Float (Alpha)"=&Multiply_Float(
      A=&FloatFromCurve(FloatCurve=DataInterface() {propertyValues {Curve {keys [{interpMode=RCIM_Cubic, tangentMode=RCTM_User, time=0, value=1}, {interpMode=RCIM_Cubic, tangentMode=RCTM_User, time=1, value=0}]}}})
      B=0.5
     )
    )
    Create Color Attribute=true
    Write to Module Namespace=true
   )
  ]
  rendererClasses [&NiagaraRibbonRendererProperties]
  renderers [{rendererClass=&NiagaraRibbonRendererProperties}]
 }}
summary {
 systemName=NS_Masamune_SwordAnimTrail_Swing
 emitters [{emitterName=Trail, rendererClasses [&NiagaraRibbonRendererProperties]}]
 userVariables [User.RibbonIDOffset {type=NiagaraInt32}]
}
